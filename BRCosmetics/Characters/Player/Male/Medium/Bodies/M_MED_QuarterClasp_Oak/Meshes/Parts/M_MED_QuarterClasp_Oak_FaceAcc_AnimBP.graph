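# Animation Blueprint for the face accessory part of the male medium QuarterClasp Oak cosmetic body, following the parent mesh's pose.
# Copy Pose From Mesh feeds a chain of bone-space Modify Bone nodes that additively scale hat, hair, head, then hat again before output.
# An extra hair scale node sits unconnected.
# The update event sets LOD1 and LOD2 flags from the current LOD, with no visible use in the AnimGraph.
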AnimBlueprint /BRCosmetics/Characters/Player/Male/Medium/Bodies/M_MED_QuarterClasp_Oak/Meshes/Parts/M_MED_QuarterClasp_Oak_FaceAcc_AnimBP
target_skeleton=&/M_MED_QuarterClasp_Oak_FaceAcc_Skeleton
graphs [
 AnimGraph {nodes [
   "Transform (Modify) Bone - Bone: hair"!(ComponentPose, bAlphaBoolEnabled:bool=True, Translation="0,0,0", Rotation="0,0,0", Scale="1.1,1.1,1.1") -> (Pose) {
    alphaInputType=Bool
    rotationSpace=BCS_BoneSpace
    scaleMode=BMM_Additive
    scaleSpace=BCS_BoneSpace
    translationMode=BMM_Additive
    translationSpace=BCS_BoneSpace
    boneToModify {boneName=hair}
    scale=<1.1 1.1 1.1>
   }
   n1 = "Transform (Modify) Bone - Bone: hat"!(
    ComponentPose=Local To Component(LocalPose=Copy Pose From Mesh() {bUseAttachedParent=true})
    bAlphaBoolEnabled:bool=True
    Translation="0,0,0"
    Rotation="0,0,0"
    Scale="1.018,1.018,1.018"
   ) -> (Pose) {
    alphaInputType=Bool
    rotationSpace=BCS_BoneSpace
    scaleMode=BMM_Additive
    scaleSpace=BCS_BoneSpace
    translationSpace=BCS_BoneSpace
    boneToModify {boneName=hat}
    scale=<1.018 1.018 1.018>
   }
   n2 = "Transform (Modify) Bone - Bone: hair"(ComponentPose=@n1, bAlphaBoolEnabled=True, Translation="0,0,0", Rotation="0,0,0", Scale="1.04,1.04,1.04") {
    alphaInputType=Bool
    rotationSpace=BCS_BoneSpace
    scaleMode=BMM_Additive
    scaleSpace=BCS_BoneSpace
    translationMode=BMM_Additive
    translationSpace=BCS_BoneSpace
    boneToModify {boneName=hair}
    scale=<1.04 1.04 1.04>
   }
   n3 = "Transform (Modify) Bone - Bone: head"!(ComponentPose=@n2, bAlphaBoolEnabled:bool=True, Translation="0,-0.5,0", Rotation="0,0,0", Scale="1.1,1.1,1.1") -> (Pose) {
    alphaInputType=Bool
    rotationSpace=BCS_BoneSpace
    scaleMode=BMM_Additive
    scaleSpace=BCS_BoneSpace
    translationSpace=BCS_BoneSpace
    boneToModify {boneName=head}
    scale=<1.1 1.1 1.1>
    translation {y=-0.5}
   }
   n4 = "Transform (Modify) Bone - Bone: hat"(ComponentPose=@n3, bAlphaBoolEnabled=True, Translation="0,-0.5,0", Rotation="0,0,0", Scale="1,1.05,1.05") {
    alphaInputType=Bool
    rotationSpace=BCS_BoneSpace
    scaleMode=BMM_Additive
    scaleSpace=BCS_BoneSpace
    translationSpace=BCS_BoneSpace
    boneToModify {boneName=hat}
    scale {y=1.05, z=1.05}
    translation {y=-0.5}
   }
   Output Pose(Result=Component To Local(ComponentPose=@n4)) {layerGroup=None}
  ]}
 EventGraph {nodes [
   Event BlueprintUpdateAnimation()
   n1 = Get CurrentLOD()
   > Set is Lod1(is Lod1:bool="Equal (Integer)"(A=@n1, B=1)) -> (Output_Get:bool)
   > Set is Lod2(is Lod2:bool="Equal (Integer)"(A=@n1, B=2)) -> (Output_Get:bool)
  ]}
]
variables [
 is Lod1 {type=boolean}
 is Lod2 {type=boolean}
]
